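# Niagara cosmetic effect for the LavaRock bass instrument, emitting glowing ash embers and soft grey smoke.
# Both sprite emitters use particle partitions to select hand-placed box regions around the instrument, as skeletal mesh sampling is disabled.
# Embers drift upward through curl noise while cooling from HDR orange to black; subUV smoke rises, grows and fades.
# Smoke spawning is gated by owner speed, while the same check on embers is switched off.
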NiagaraSystem /SparksCosmetics/Instrument/Bass/LavaRock/Effects/NS_Bass_LavaRock
dependencies {
 usedDataInterfaces [
  &NiagaraDataInterfaceArrayFloat3
  &NiagaraDataInterfaceArrayFloat
  &NiagaraDataInterfaceArrayInt32
  &NiagaraDataInterfaceCurve
  &NiagaraDataInterfaceColorCurve
  &NiagaraDataInterfaceVector2DCurve
  &NiagaraDataInterfaceSpriteRendererInfo
  &NiagaraDataInterfaceSkeletalMesh
 ]
 usedDynamicInputs [&Vector2DFromCurve, &RandomRangeVector2D]
 usedModules [
  &self:SystemSpawnScript.SystemScriptSource.SystemScriptGraph.NiagaraNodeAssignment_0.SetVariables_#1
  &SystemState
 ]
}
emitters {
 Ashes001 {
  emitterUpdateScript [
   &EmitterState(
    Life Cycle Mode=::NewEnumerator1
    Loop Duration=&RandomRangeFloat(Minimum=0.65, Maximum=4)
    Recalculate Duration Each Loop=true
    UseLoopDelay=false
   )
   &SpawnRate(
    Spawn Probability=&MakeCustomFloatFromBool(
     Bool=&SetBoolByFloatComparison(A=&VectorLength(Vector=$Engine.Owner.Velocity), B=200, Comparison Type=::NewEnumerator6)
     True Float=1
     False Float=0
    )
    SpawnRate=&RandomRangeFloat(Minimum=10, Maximum=15)
    Use Spawn Probability=false
   )
  ]
  particleSpawnScript [
   &PartitionParticles(Number Of Partitions=6)
   &InitializeParticle(
    Color=<rgba 20 20 20 1>
    Lifetime=&RandomRangeFloat(Minimum=0.75, Maximum=1.5)
    Position Offset=&SelectVectorFromArray(
     Direct Array Index=$Particles.PartitionParticles.Partition
     Vector Selection Array=DataInterface() {propertyValues {FloatData [<2.2 0 -24.4>, <-2.5 -1 -24.4>, <8 -3 0.2>, <-8 -3 17.4>, <-0.4 -1 94.3>, <-2.2 -1 12.5>]}}
    )
    Sprite Rotation Mode=::NewEnumerator2
    Sprite Size Mode=::NewEnumerator1
    Uniform Sprite Size Max=3.25
    Uniform Sprite Size Min=1.75
    UsePositionOffset=true
    Write Material Random=true
   )
   &ShapeLocation(
    Box Midpoint=&SelectVectorFromArray(
     Direct Array Index=$Particles.PartitionParticles.Partition
     Vector Selection Array=DataInterface() {propertyValues {FloatData [<0.5 0.5 0>, <0 0.5 0>, <0 0.5 0.5>, <0.4 0.5 0.5>, <0.5 0.5 0.5>, <0.4 0.5 0.5>]}}
    )
    Box Size=&SelectVectorFromArray(
     Direct Array Index=$Particles.PartitionParticles.Partition
     Vector Selection Array=DataInterface() {propertyValues {FloatData [<3 9 10>, <3 9 30>, <3 9 30>, <6 14 22>, <4 5 4>, <6 15 10>]}}
    )
    MaskSpawnGroup=false
    Override Seed=false
    Rotation Angle=90
    Rotation Angle Type:ENiagara_AngleInput=::NewEnumerator0
    Rotation Axis=<0 0 1>
    Rotation Mode=::NewEnumerator1
    Shape Origin=$Particles.Position
    Shape Primitive=::NewEnumerator2
    Surface Only=true
    Surface Only Band Thickness=0.5
    UseOwnerScale=false
   ) {module=ShapeLocation001}
   /SetVariables(Particles.Position=&AddVectorToPosition(
     Position=$Particles.Position
     Vector=&TransformVector(
      Vector=&SelectVectorFromArray(
       Direct Array Index=$Particles.PartitionParticles.Partition
       Vector Selection Array=DataInterface() {propertyValues {FloatData [<-2.2 0 -24.4>, <-2.2 0 -24.4>, <-2.2 -7 0.2>, <-2.2 9.4 17.4>, <-0.4 -1 94.3>, <-2.2 -9.3 12.5>]}}
      )
      Source Space=::World
      Destination Space=::Simulation
     )
    )) {enabled=false, isSetParameters=true}
  ]
  particleUpdateScript [
   &ParticleState()
   &SkeletalMeshLocation(
    Bone / Socket Index=&SelectIntFromArray(Direct Array Index=$Particles.PartitionParticles.Partition, Int Selection Array=DataInterface() {propertyValues {IntData [0,0,1,2,3,4]}})
    Bone / Socket Sampling Mode=::NewEnumerator5
    MaskSpawnGroup=false
    Mesh Sampling Type=::NewEnumerator6
    Position Sampling=::NewEnumerator1
    Sampled Position Offset=$StackContext.ShapeLocation001.ShapePosition
    Skeletal Mesh=$System.Skeletal Mesh
   ) {enabled=false}
   &AccelerationForce(Acceleration=&RandomRangeVector(Minimum=<0 0 8>, Maximum=<0 0 15>)) {module=AccelerationForce001}
   &CurlNoiseForce(Noise Frequency=100, Noise Strength=5, Pan Noise=false)
   &SolveForcesAndVelocity(Clamp Velocity=true, Limit Acceleration=true)
   &ScaleSpriteSize(Uniform Curve Sprite Scale=DataInterface() {propertyValues {Curve {keys [
        {arriveTangent=5.5195, interpMode=RCIM_Cubic, leaveTangent=5.5195, tangentMode=RCTM_User, time=0, value=0}
        {arriveTangentWeight=0.25, leaveTangentWeight=0.0932, time=0.75, value=1}
        {arriveTangentWeight=0.4331, time=1, value=0}
       ]}}})
   &ScaleColor(
    Scale RGB=&MakeVectorFromLinearColorRGB(Color=&ColorFromCurve(ColorCurve=DataInterface() {propertyValues {
        AlphaCurve {keys []}
        BlueCurve {keys [
          {interpMode=RCIM_Cubic, leaveTangentWeight=0.2267, tangentMode=RCTM_User, time=0, value=49.0353}
          {arriveTangentWeight=11.117, interpMode=RCIM_Cubic, leaveTangentWeight=5.2315, tangentMode=RCTM_User, time=0.68, value=0}
          {arriveTangentWeight=0.1067, interpMode=RCIM_Cubic, tangentMode=RCTM_User, time=1, value=0}
         ]}
        GreenCurve {keys [
          {interpMode=RCIM_Cubic, leaveTangentWeight=0.2267, tangentMode=RCTM_User, time=0, value=154.5329}
          {arriveTangentWeight=35.0282, interpMode=RCIM_Cubic, leaveTangentWeight=16.4839, tangentMode=RCTM_User, time=0.68, value=0.426}
          {arriveTangentWeight=0.1067, interpMode=RCIM_Cubic, tangentMode=RCTM_User, time=1, value=0}
         ]}
        RedCurve {keys [
          {interpMode=RCIM_Cubic, leaveTangentWeight=0.2267, tangentMode=RCTM_User, time=0, value=500}
          {arriveTangentWeight=113.3336, interpMode=RCIM_Cubic, leaveTangentWeight=53.3334, tangentMode=RCTM_User, time=0.68, value=10}
          {arriveTangentWeight=0.1067, interpMode=RCIM_Cubic, tangentMode=RCTM_User, time=1, value=0}
         ]}
       }}))
    ScaleA=true
    ScaleRGB=true
   )
   &SpriteRotationRate(Rotation Rate=&RandomRangeFloat(Minimum=-3, Maximum=3))
   /SetVariables(Particles.Position=&AddVectorToPosition(Position=$Particles.Position, Vector=&TransformVector(Vector=<0 0 0.25>, Destination Space=::World))) {enabled=false, isSetParameters=true}
  ]
  rendererClasses [&NiagaraSpriteRendererProperties]
  renderers [{rendererClass=&NiagaraSpriteRendererProperties}]
 }
 Smoke_Detail_Spotlight {
  emitterUpdateScript [
   &EmitterState(Inactive Response=::NewEnumerator1, Life Cycle Mode=::NewEnumerator1, Loop Delay=0.4, Loop Duration=1, UseLoopDelay=false)
   &SpawnRate(
    Spawn Probability=&MakeCustomFloatFromBool(
     Bool=&SetBoolByFloatComparison(A=&VectorLength(Vector=$Engine.Owner.Velocity), B=200, Comparison Type=::NewEnumerator6)
     True Float=1
     False Float=0
    )
    SpawnRate=&RandomRangeFloat(Minimum=5, Maximum=10)
    Use Spawn Probability=true
   ) {module=SpawnRate001}
  ]
  particleSpawnScript [
   &PartitionParticles(Number Of Partitions=5)
   &InitializeParticle(
    Color=<rgba 0.2353 0.2353 0.2353 0.325>
    Lifetime=&Multiply_Float(
     A=&SelectFloatFromArray(Direct Array Index=$Particles.PartitionParticles.Partition, Float Selection Array=DataInterface() {propertyValues {FloatData [2,3.5,3.5]}})
     B=&RandomRangeFloat(Minimum=0.6, Maximum=1)
    )
    Sprite Rotation Mode=::NewEnumerator2
    Sprite Size Max=<17 15>
    Sprite Size Min=<12 12>
    Sprite Size Mode=::NewEnumerator2
    UsePositionOffset=false
    Write Material Random=true
   )
   /SetVariables(Particles.Position=&AddVectorToPosition(
     Position=$Particles.Position
     Vector=&SelectVectorFromArray(
      Direct Array Index=$Particles.PartitionParticles.Partition
      Vector Selection Array=DataInterface() {propertyValues {FloatData [<-2.2 0 -15>, <-2.2 0 -15>, <-2.2 -3 0.2>, <-2.2 -3 17.4>, <-0.4 0 94.3>, <-2.2 0 12.5>]}}
     )
    )) {isSetParameters=true}
   &ShapeLocation(
    Box Midpoint=&SelectVectorFromArray(
     Direct Array Index=$Particles.PartitionParticles.Partition
     Vector Selection Array=DataInterface() {propertyValues {FloatData [<0.5 0.5 0>, <0 0.5 0>, <0 0.5 0.5>, <0.4 0.5 0.5>, <0.5 0.5 0.5>, <0.4 0.5 0.5>]}}
    )
    Box Size=&SelectVectorFromArray(
     Direct Array Index=$Particles.PartitionParticles.Partition
     Vector Selection Array=DataInterface() {propertyValues {FloatData [<5 9 10>, <5 9 5>, <5 9 30>, <7 14 22>, <5 5 4>, <7 15 10>]}}
    )
    MaskSpawnGroup=false
    Override Seed=false
    Rotation Mode=::NewEnumerator1
    Shape Origin=$Particles.Position
    Shape Primitive=::NewEnumerator2
    Surface Only=true
    Surface Only Band Thickness=0.5
    UseOwnerScale=true
   ) {module=ShapeLocation001}
   &ShapeLocation(
    Box Midpoint=&SelectVectorFromArray(
     Direct Array Index=$Particles.PartitionParticles.Partition
     Vector Selection Array=DataInterface() {propertyValues {FloatData [<0.5 0.5 0>, <0.5 0.5 0.5>, <0.5 0.5 0.5>, <0.5 0.5 0.5>]}}
    )
    Box Size=&SelectVectorFromArray(
     Direct Array Index=$Particles.PartitionParticles.Partition
     Vector Selection Array=DataInterface() {propertyValues {FloatData [<1 1 20>, <1 1 10>, <1 1 20>, <1 1 1>]}}
    )
    MaskSpawnGroup=false
    Override Seed=false
    Rotation Angle=90
    Rotation Angle Type:ENiagara_AngleInput=::NewEnumerator0
    Rotation Axis=<0 0 1>
    Rotation Mode=::NewEnumerator1
    Shape Origin=$Particles.Position
    Shape Primitive=::NewEnumerator2
    Surface Only=true
    Surface Only Band Thickness=0.5
    UseOwnerScale=true
   ) {enabled=false}
  ]
  particleUpdateScript [
   &ParticleState()
   &SkeletalMeshLocation(
    Bone / Socket Index=&SelectIntFromArray(Direct Array Index=$Particles.PartitionParticles.Partition, Int Selection Array=DataInterface() {propertyValues {IntData [0,1,2,3]}})
    Bone / Socket Sampling Mode=::NewEnumerator5
    MaskSpawnGroup=false
    Mesh Sampling Type=::NewEnumerator6
    Position Sampling=::NewEnumerator1
    Sampled Position Offset=$StackContext.ShapeLocation.ShapePosition
    Skeletal Mesh=$System.Skeletal Mesh
   ) {enabled=false}
   &SpriteRotationRate(Rotation Rate=&RandomRangeFloat(Minimum=-15, Maximum=15))
   &ScaleSpriteSize(
    Scale Sprite Size Mode=::NewEnumerator0
    Uniform Scale Factor=&FloatFromCurve(FloatCurve=DataInterface() {propertyValues {Curve {keys [<at 0 0.8>,<at 1 1.2>]}}}, Scale Curve=1.3)
   )
   &SubUVAnimation(
    End Frame Range Override=30
    "Sprite Renderer "=DataInterface() {propertyValues {SpriteRenderer=&self:Smoke_Detail_Spotlight_3.NiagaraSpriteRendererProperties_0}}
    Start Frame Range Override=&RandomRangeInt(Minimum=0, Maximum=5)
    UseEndFrame=true
    UseStartFrame=true
   )
   &ScaleColor(
    Scale Alpha=&Multiply_Float(
     A=&RandomRangeFloat(Minimum=0.5, Maximum=1)
     B=&FloatFromCurve(FloatCurve=DataInterface() {propertyValues {Curve {keys [
          {arriveTangent=9.0725, interpMode=RCIM_Cubic, leaveTangent=9.0725, tangentMode=RCTM_User, time=0, value=0}
          {arriveTangent=0.0706, interpMode=RCIM_Cubic, leaveTangent=0.0706, tangentMode=RCTM_User, time=0.1408, value=0.9847}
          {arriveTangent=-0.0405, interpMode=RCIM_Cubic, leaveTangent=-0.0405, tangentMode=RCTM_User, time=1, value=0}
         ]}}})
    )
    Scale RGB=&MakeVectorFromLinearColorRGB(Color=&ColorFromCurve(ColorCurve=DataInterface() {propertyValues {
        AlphaCurve {keys [<at 0 1 RCIM_Cubic>,<at 1 0 RCIM_Cubic>]}
        BlueCurve {keys [<at 0 1 RCIM_Cubic>,<at 0.702 0.2292 RCIM_Cubic>]}
        GreenCurve {keys [<at 0 1 RCIM_Cubic>,<at 0.702 0.2292 RCIM_Cubic>]}
        RedCurve {keys [<at 0 1 RCIM_Cubic>,<at 0.702 0.2292 RCIM_Cubic>]}
       }}))
    ScaleA=true
    ScaleRGB=true
   )
   &GravityForce(Gravity=&Multiply_Vector(
     A=&RandomRangeVector(Minimum=<0 0 25>, Maximum=<0 0 50>)
     B=&MakeVector(
      X=1
      Y=1
      Z=&FloatFromCurve(FloatCurve=DataInterface() {propertyValues {Curve {keys [
           {arriveTangent=0.6752, interpMode=RCIM_Cubic, leaveTangent=0.6752, tangentMode=RCTM_User, time=0, value=1}
           {interpMode=RCIM_Cubic, tangentMode=RCTM_User, time=0.2, value=0}
          ]}}})
     )
    ))
   &Drag(Use Linear Drag=true, Use Rotational Drag=true)
   &SolveForcesAndVelocity(Clamp Velocity=false, Limit Acceleration=false)
   &CameraOffset(Camera Offset Amount=2)
   /SetVariables(Particles.Position=&AddVectorToPosition(Position=$Particles.Position, Vector=&TransformVector(Vector=<0 0 0.25>, Destination Space=::World))) {enabled=false, isSetParameters=true}
  ]
  rendererClasses [&NiagaraSpriteRendererProperties]
  renderers [{rendererClass=&NiagaraSpriteRendererProperties}]
 }
}
summary {
 systemName=NS_Bass_LavaRock
 emitters [
  {emitterName=Ashes001, rendererClasses [&NiagaraSpriteRendererProperties]}
  {emitterName=Smoke_Detail_Spotlight, rendererClasses [&NiagaraSpriteRendererProperties]}
 ]
 userVariables []
}
